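# Physics asset for the BillyGold backpack cosmetic, defining collision and joints for a hanging handcuffs chain.
# Two boxes on the pack bone form the body that the chain's top capsule attaches to.
# Seven small capsules continue from the top link to a larger box on the last handcuff bone.
# Each link is constrained to the one before it, with tight swing and twist limits that loosen toward the free end.
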PhysicsAsset /BRCosmetics/Accessories/FORT_Backpacks/Backpack_BillyGold/Meshes/M_MED_BillyGold_Pack_PhysicsAsset
bodies {
 dyn_handcuffs_01 [{shapeType=Capsule, length=1, radius=1, shapeName=dyn_handcuffs_01_capsule, center=<1 0 0>, rotation=<rot 90 0 0>}]
 dyn_handcuffs_02 [{shapeType=Capsule, length=1, radius=1, shapeName=dyn_handcuffs_02_capsule, center=<1 0 0>, rotation=<rot 90 0 0>}]
 dyn_handcuffs_03 [{shapeType=Capsule, length=1, radius=1, shapeName=dyn_handcuffs_03_capsule, center=<1 0 0>, rotation=<rot 90 0 0>}]
 dyn_handcuffs_04 [{shapeType=Capsule, length=1, radius=1, shapeName=dyn_handcuffs_04_capsule, center=<1 0 0>, rotation=<rot 90 0 0>}]
 dyn_handcuffs_05 [{shapeType=Capsule, length=1, radius=1, shapeName=dyn_handcuffs_05_capsule, center=<1 0 0>, rotation=<rot 90 0 0>}]
 dyn_handcuffs_06 [{shapeType=Capsule, length=1, radius=1, shapeName=dyn_handcuffs_06_capsule, center=<1 0 0>, rotation=<rot 90 0 0>}]
 dyn_handcuffs_07 [{shapeType=Capsule, length=1, radius=1, shapeName=dyn_handcuffs_07_capsule, center=<1 0 0>, rotation=<rot 90 0 0>}]
 dyn_handcuffs_08 [{shapeType=Box, extentX=10, extentY=3, extentZ=14, shapeName=dyn_handcuffs_08_box, center=<9 0 0>, rotation=<rot 0 -100 -90>}]
 dyn_handcuffs_Top [{
   shapeType=Capsule
   length=3
   radius=2
   shapeName=dyn_handcuffs_Top_capsule
   center=<2.9684 0 0.5565>
   rotation=<rot 80.9924 179.9994 179.9994>
  }]
 pack [
  {
   shapeType=Box
   extentX=11.7391
   extentY=19.8916
   extentZ=27.7348
   shapeName=pack_box
   center=<-0.0008 -5.8597 3.8326>
   rotation=<rot 0 89.9892 -89.9838>
  }
  {shapeType=Box, extentX=31.9581, extentY=16.2699, extentZ=26.9628, shapeName=pack_box, center=<-21.3728 3.8521 4.3878>}
 ]
}
constraints [
 {bone1Name=dyn_handcuffs_Top, bone2Name=pack, swing1LimitDegrees=12, swing2LimitDegrees=4, twistLimitDegrees=3}
 {bone1Name=dyn_handcuffs_01, bone2Name=dyn_handcuffs_Top, swing1LimitDegrees=8, swing2LimitDegrees=5, twistLimitDegrees=3}
 {bone1Name=dyn_handcuffs_02, bone2Name=dyn_handcuffs_01, swing1LimitDegrees=8, swing2LimitDegrees=5, twistLimitDegrees=3}
 {bone1Name=dyn_handcuffs_03, bone2Name=dyn_handcuffs_02, swing1LimitDegrees=8, swing2LimitDegrees=5, twistLimitDegrees=3}
 {bone1Name=dyn_handcuffs_04, bone2Name=dyn_handcuffs_03, swing1LimitDegrees=8, swing2LimitDegrees=5, twistLimitDegrees=3}
 {bone1Name=dyn_handcuffs_05, bone2Name=dyn_handcuffs_04, swing1LimitDegrees=8, swing2LimitDegrees=5, twistLimitDegrees=5}
 {bone1Name=dyn_handcuffs_06, bone2Name=dyn_handcuffs_05, swing1LimitDegrees=30, swing2LimitDegrees=5, twistLimitDegrees=5}
 {bone1Name=dyn_handcuffs_07, bone2Name=dyn_handcuffs_06, swing1LimitDegrees=30, swing2LimitDegrees=5, twistLimitDegrees=5}
 {bone1Name=dyn_handcuffs_08, bone2Name=dyn_handcuffs_07, swing1LimitDegrees=45, swing2LimitDegrees=15, twistLimitDegrees=5}
]
